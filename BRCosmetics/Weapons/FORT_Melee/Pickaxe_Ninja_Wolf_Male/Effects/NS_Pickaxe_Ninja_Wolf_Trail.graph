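# Niagara trail effect for the Ninja Wolf male pickaxe, drawing red-to-purple flame ribbons and eye flame sprites from a user-supplied skeletal mesh.
# FlameTrails uses the anim trail modules to build fading ribbons, scrolling U with engine time and animating material parameters.
# EyeFlames spawns looping bursts of rotating flipbook sprites on the mesh that inherit velocity, slow with drag and shrink.
# A thinner second ribbon emitter, FlameTrails001, is disabled.
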NiagaraSystem /BRCosmetics/Weapons/FORT_Melee/Pickaxe_Ninja_Wolf_Male/Effects/NS_Pickaxe_Ninja_Wolf_Trail
dependencies {
 usedDataInterfaces [&NiagaraDataInterfaceColorCurve, &NiagaraDataInterfaceCurve, &NiagaraDataInterfaceVectorCurve]
 usedDynamicInputs [&VectorFromCurve]
 usedModules [&SystemState]
}
emitters {
 EyeFlames {
  emitterUpdateScript [
   &EmitterState(Life Cycle Mode=::NewEnumerator1, Loop Behavior=::NewEnumerator1, Loop Delay=0.05, Loop Duration=0.14, UseLoopDelay=true)
   &SpawnRate(SpawnRate=100, Use Spawn Probability=false)
  ]
  particleSpawnScript [
   /Niagara/Modules/Spawn/Initialization/InitializeParticle(
    Color=<rgba 1 1 1 0.05>
    Lifetime=&UniformRangedFloat(Minimum=0.3, Maximum=0.6)
    Module.Write Color=true
    Module.Write Lifetime=true
    Module.Write Mass=false
    Module.Write Material Random=true
    Module.Write Position=false
    Module.Write Scale=false
    Module.Write SpriteRotation=true
    Module.Write SpriteSize=true
    Position=&SimulationPosition()
    Sprite Rotation=&UniformRangedFloat(Minimum=0, Maximum=360)
    Sprite Size=&Vector2DFromFloat(Value=&RandomRangeFloat(Minimum=10, Maximum=20))
   )
   &DynamicMaterialParameters(
    Index 0 Param 2=&UniformRangedFloat(Minimum=0, Maximum=1)
    Index 0 Param 3=1
    Index 0 Param 4=&UniformRangedFloat(Minimum=1, Maximum=1)
    Module.Param0WriteEnabled=false
    Module.Param1WriteEnabled=true
    Module.Param2WriteEnabled=false
    Module.Param3WriteEnabled=true
   ) {module=DynamicMaterialParameters002}
   &FN_SetSpawnGroup(Number of Spawn Groups=2)
   &CameraOffset(Camera Offset Amount=&RandomRangeFloat(Minimum=5, Maximum=7))
   &SkeletalMeshLocation(MaskSpawnGroup=false, Mesh Sampling Type=::NewEnumerator5, Skeletal Mesh=$User.SkeletalMesh, Socket Sampling Mode=::NewEnumerator5)
   &InheritVelocity(Inherited Velocity Amount Scale=<2 2 2>, Inherited Velocity Speed Limit=200, Speed Limit=true, UseSpeedThreshold=false)
  ]
  particleUpdateScript [
   &Drag(Drag=6, Use Linear Drag=true, Use Rotational Drag=false)
   &SolveForcesAndVelocity(Clamp Velocity=false, Limit Acceleration=false)
   &ParticleState()
   &ScaleColor(
    Scale Mode=::NewEnumerator1
    Scale RGBA=&Vector4FromCurve(Vector4Curve=DataInterface() {propertyValues {
       AlphaCurve {keys [<at 0 0>,<at 0.1513 10>,<at 1 0>]}
       BlueCurve {keys [<at 0.0312 0.1717>,<at 0.1903 0.1244>,<at 0.4242 0.2>,<at 0.917 0.1>]}
       GreenCurve {keys [<at 0.0312 0>,<at 0.1903 0.0135>,<at 0.4242 0>,<at 0.917 0.003>]}
       RedCurve {keys [<at 0.0312 1>,<at 0.1903 0.1719>,<at 0.4242 0.0967>,<at 0.917 0.0555>]}
      }})
    ScaleRGBA=true
   )
   /SetVariables(Particles.SubImageIndex=&FloatFromCurve(FloatCurve=DataInterface() {propertyValues {Curve {keys [<at 0 1>,<at 1 32>]}}}, Scale Curve=2)) {isSetParameters=true}
   &ScaleSpriteSize(Scale Factor=&Vector2DFromFloat(Value=&FloatFromCurve(FloatCurve=DataInterface() {propertyValues {Curve {keys [
          {arriveTangent=4.8245, interpMode=RCIM_Cubic, leaveTangent=4.8245, tangentMode=RCTM_User, time=0, value=0.25}
          {arriveTangent=-0.25, interpMode=RCIM_Cubic, leaveTangent=-0.25, time=0.1, value=1.4}
          <at 1 0 RCIM_Cubic>
         ]}}})))
   &DynamicMaterialParameters(
    Index 0 Param 1=&FloatFromCurve(FloatCurve=DataInterface() {propertyValues {Curve {keys [{arriveTangent=2.7551, interpMode=RCIM_Cubic, leaveTangent=2.7551, tangentMode=RCTM_User, time=0, value=0}, <at 0.5 1>]}}})
    Index 0 Param 2=&RandomRangeFloat(Minimum=0.33, Maximum=1)
    Index 0 Param 3=&FloatFromCurve(FloatCurve=DataInterface() {propertyValues {Curve {keys [<at 0.2 0>,<at 1 -40>]}}})
    Index 0 Param 4=1.2
    Module.Param0WriteEnabled=true
    Module.Param1WriteEnabled=true
    Module.Param2WriteEnabled=true
    Module.Param3WriteEnabled=true
   ) {module=DynamicMaterialParameters001}
   &SpriteRotationRate(Rotation Rate=100)
   &Color() {enabled=false}
  ]
  rendererClasses [&NiagaraSpriteRendererProperties]
  renderers [{rendererClass=&NiagaraSpriteRendererProperties}]
 }
 FlameTrails {
  emitterUpdateScript [
   &EmitterState(Life Cycle Mode=::NewEnumerator1, Loop Behavior=::NewEnumerator1, Loop Delay=0.5, Loop Duration=0, UseLoopDelay=true)
   &FN_SpawnAnimTrail()
  ]
  particleSpawnScript [
   &FN_InitializeAnimTrailBoneIndecies()
   &FN_InitializeAnimTrail(Color=<rgba 10 10 10 0.5>, Lifetime=0.4)
   /SetVariables(Particles.FrontRibbonMask:NiagaraFloat=&OneMinusFloat(Float=/Niagara/DynamicInputs/Integer/V2/Round_Float(Float:NiagaraFloat=$System.NormalizedLoopAge))) {isSetParameters=true, module=SetVariables_#1}
   &InitializeRibbon(Ribbon Width=20, UsePositionOffset=false, Write Material Random=true) {enabled=false}
  ]
  particleUpdateScript [
   &FN_UpdateAnimTrail(SkeletalMesh=$User.SkeletalMesh)
   &ScaleColor(
    Scale Alpha=&FloatFromCurve(
     FloatCurve=DataInterface() {propertyValues {Curve {keys [{tangentMode=RCTM_User, time=0, value=1}, {tangentMode=RCTM_User, time=1, value=0}]}}}
     Scale Curve=0.5
    )
    Scale RGB=&MakeVectorFromLinearColorRGB(Color=&ColorFromCurve(ColorCurve=DataInterface() {propertyValues {
        AlphaCurve {keys [<at 0.1513 3>,<at 1 0>]}
        BlueCurve {keys [<at 0.0062 1.2345>,<at 0.1743 0.0801>,<at 0.2777 0.5>,<at 0.4644 1>,<at 0.9241 0.1>]}
        GreenCurve {keys [<at 0.0062 0>,<at 0.1743 0>,<at 0.2777 0>,<at 0.4644 0>,<at 0.9241 0.003>]}
        RedCurve {keys [<at 0.0062 10>,<at 0.1743 0.3542>,<at 0.2777 0.4158>,<at 0.4644 0.3724>,<at 0.9241 0.0555>]}
       }}))
    ScaleA=true
    ScaleRGB=true
   )
   &ParticleState()
   &ScaleRibbonWidth(Ribbon Width Scale=&FloatFromCurve(
     FloatCurve=DataInterface() {propertyValues {Curve {keys [
         {arriveTangent=8.6142, interpMode=RCIM_Cubic, leaveTangent=8.6142, tangentMode=RCTM_User, time=0, value=0}
         {arriveTangent=-0.156, interpMode=RCIM_Cubic, leaveTangent=-0.156, tangentMode=RCTM_User, time=0.2, value=0.95}
         {arriveTangent=-2.3748, interpMode=RCIM_Cubic, leaveTangent=-2.3748, tangentMode=RCTM_User, time=0.9, value=0}
        ]}}}
     Scale Curve=1.2
    ))
   /SetVariables(Particles.RibbonU0Override=&Multiply_Float(A=$Engine.Time, B=4)) {isSetParameters=true, module=SetVariables_#2}
   &DynamicMaterialParameters(
    Index 0 Param 1=&FloatFromCurve(FloatCurve=DataInterface() {propertyValues {Curve {keys [<at 0 1>,<at 1 0>]}}})
    Index 0 Param 3=2
    Index 0 Param 4=&FloatFromCurve(FloatCurve=DataInterface() {propertyValues {Curve {keys [{tangentMode=RCTM_User, time=0.3, value=0}, {tangentMode=RCTM_User, time=0.9, value=1}]}}})
    Index 1 Param 4=&FloatFromCurve(FloatCurve=DataInterface() {propertyValues {Curve {keys [<at 0.3 0.04>,<at 1 0.15>]}}})
    Module.Param0WriteEnabled=true
    Module.Param0WriteEnabled1=true
    Module.Param1WriteEnabled=true
    Module.Param1WriteEnabled1=false
    Module.Param2WriteEnabled=true
    Module.Param2WriteEnabled1=true
    Module.Param3WriteEnabled=true
    Module.Param3WriteEnabled1=false
    Write Parameter Index 1=true
   )
   &CameraOffset(Camera Offset Amount=10)
  ]
  rendererClasses [&NiagaraRibbonRendererProperties]
  renderers [{rendererClass=&NiagaraRibbonRendererProperties}]
 }
 FlameTrails001 {
  bEnabled=false
  emitterUpdateScript [
   &EmitterState(Life Cycle Mode=::NewEnumerator1, Loop Behavior=::NewEnumerator1, Loop Delay=0.5, Loop Duration=0, UseLoopDelay=true)
   &FN_SpawnAnimTrail()
  ]
  particleSpawnScript [
   &FN_InitializeAnimTrailBoneIndecies()
   &FN_InitializeAnimTrail(AnimTrailWidth=2, Color=<rgba 10 10 10 0.1>)
   /SetVariables(Particles.FrontRibbonMask:NiagaraFloat=&OneMinusFloat(Float=&Round_Float(Float:NiagaraFloat=$System.NormalizedLoopAge))) {isSetParameters=true, module=SetVariables_#1}
  ]
  particleUpdateScript [
   &FN_UpdateAnimTrail(SkeletalMesh=$User.SkeletalMesh)
   &ScaleColor(
    Scale Alpha=&FloatFromCurve(
     FloatCurve=DataInterface() {propertyValues {Curve {keys [{tangentMode=RCTM_User, time=0, value=1}, {tangentMode=RCTM_User, time=1, value=0}]}}}
     Scale Curve=0.5
    )
    Scale RGB=&MakeVectorFromLinearColorRGB(Color=&ColorFromCurve(ColorCurve=DataInterface() {propertyValues {
        AlphaCurve {keys [<at 0.1513 3>,<at 1 0>]}
        BlueCurve {keys [<at 0.0062 1.2345>,<at 0.0602 0.1608>,<at 0.3579 0.5>,<at 0.917 0.1>]}
        GreenCurve {keys [<at 0.0062 0>,<at 0.0602 0>,<at 0.3579 0>,<at 0.917 0.003>]}
        RedCurve {keys [<at 0.0062 10>,<at 0.0602 0.3542>,<at 0.3579 0.4158>,<at 0.917 0.0555>]}
       }}))
    ScaleA=true
    ScaleRGB=true
   )
   &ParticleState()
   &ScaleRibbonWidth(Ribbon Width Scale=&FloatFromCurve(
     FloatCurve=DataInterface() {propertyValues {Curve {keys [
         {arriveTangent=-0.156, interpMode=RCIM_Cubic, leaveTangent=-0.156, tangentMode=RCTM_User, time=0, value=1}
         {arriveTangent=-1.7494, interpMode=RCIM_Cubic, leaveTangent=-1.7494, tangentMode=RCTM_User, time=1, value=0.25}
        ]}}}
     Scale Curve=0.3
    ))
   /SetVariables(Particles.RibbonU0Override=&Multiply_Float(A=$Engine.Time, B=4)) {isSetParameters=true, module=SetVariables_#2}
   &DynamicMaterialParameters(
    Index 1 Param 4=&FloatFromCurve(FloatCurve=DataInterface() {propertyValues {Curve {keys [<at 0.3 0.04>,<at 1 0.15>]}}})
    Module.Param0WriteEnabled1=false
    Module.Param1WriteEnabled1=false
    Module.Param2WriteEnabled1=false
    Module.Param3WriteEnabled1=false
    Write Parameter Index 0=false
    Write Parameter Index 1=true
   )
   &CameraOffset(Camera Offset Amount=10)
  ]
  rendererClasses [&NiagaraRibbonRendererProperties]
  renderers [{rendererClass=&NiagaraRibbonRendererProperties}]
 }
}
summary {
 systemName=NS_Pickaxe_Ninja_Wolf_Trail
 emitters [
  {emitterName=FlameTrails, rendererClasses [&NiagaraRibbonRendererProperties]}
  {emitterName=EyeFlames, rendererClasses [&NiagaraSpriteRendererProperties]}
  {bEnabled=false, emitterName=FlameTrails001, rendererClasses [&NiagaraRibbonRendererProperties]}
 ]
 userVariables [
  User.LifetimeCustom {type=NiagaraFloat, defaultValue=0, description=""}
  User.LightColor {type=LinearColor, description="", defaultValue {struct=LinearColor, value=<rgba 1 1 1 1>}}
  User.RibbonIDOffset {type=NiagaraInt32}
  User.SkeletalMesh {
   type=NiagaraDataInterfaceSkeletalMesh
   defaultValue {
    struct=NiagaraExt_VariableValue_DataInterface
    value {dataInterface=&self:NiagaraDataInterfaceSkeletalMesh_0}
   }
  }
 ]
}
