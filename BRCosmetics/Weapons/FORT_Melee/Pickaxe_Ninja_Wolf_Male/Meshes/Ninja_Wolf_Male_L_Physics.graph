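# Physics asset for the Ninja Wolf Male pickaxe's L mesh, apparently the left-hand piece, giving its dangling strap bones physical bodies.
# A small capsule on the handle bone anchors three identical strap capsules, each offset along its bone's X axis.
# Constraints link them in a serial chain from handle to strap tip, with matching moderate swing limits and tight twist limits.
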PhysicsAsset /BRCosmetics/Weapons/FORT_Melee/Pickaxe_Ninja_Wolf_Male/Meshes/Ninja_Wolf_Male_L_Physics
bodies {
 Melee_TwoHanded_Handle [{shapeType=Capsule, length=1, radius=1}]
 dyn_strap1 [{shapeType=Capsule, length=2, radius=2, center=<5 0 0>}]
 dyn_strap2 [{shapeType=Capsule, length=2, radius=2, center=<5 0 0>}]
 dyn_strap3 [{shapeType=Capsule, length=2, radius=2, center=<5 0 0>}]
}
constraints [
 {bone1Name=dyn_strap1, bone2Name=Melee_TwoHanded_Handle, swing1LimitDegrees=45, swing2LimitDegrees=45, twistLimitDegrees=10}
 {bone1Name=dyn_strap2, bone2Name=dyn_strap1, swing1LimitDegrees=45, swing2LimitDegrees=45, twistLimitDegrees=10}
 {bone1Name=dyn_strap3, bone2Name=dyn_strap2, swing1LimitDegrees=45, swing2LimitDegrees=45, twistLimitDegrees=10}
]
